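# Control Rig for the Tactical Fisherman medium male body that orients the rope hanging from the pelvis-mounted carabiner.
# An Aim Bone unit points rope_aim at rope_aim_tgt, using rope_aim_up as the secondary target, and propagates to the rope chain.
# Both targets descend from thigh_twist_01_r, so the rope's orientation follows the right thigh.
# A debug axes draw of rope_aim runs before the aim step.
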ControlRigBlueprint /Game/Characters/Player/Male/Medium/Bodies/M_MED_Tactical_Fisherman/Meshes/M_MED_Tactical_Fisherman_CtrlRig
graphs [
 RigVMFunctionLibrary {links [], nodes []}
 RigVMModel {nodes [
   n1 = &RigUnit_BeginExecution()
   /Script/ControlRig.RigUnit_DebugTransformMutable(
    Color=(R=1,G=0,B=0,A=1)
    DebugDrawSettings=(DepthPriority=SDPG_Foreground,Lifetime=-1)
    Mode=Axes
    Scale=10
    Space=None
    Thickness=0
    WorldOffset=(Rotation=(X=0,Y=0,Z=0,W=1),Translation=(X=0,Y=0,Z=0),Scale3D=(X=1,Y=1,Z=1))
    bEnabled=True
    ExecutePin=@n1.ExecutePin
    Transform=&RigUnit_GetBoneTransform(Bone=rope_aim).Transform
   ) {is=struct RigVMUnitNode}
   > /Script/ControlRig.RigUnit_AimBone(
    Bone=rope_aim
    DebugSettings=(bEnabled=true,Scale=10,WorldOffset=(Rotation=(X=0,Y=0,Z=0,W=1),Translation=(X=0,Y=0,Z=0),Scale3D=(X=1,Y=1,Z=1)))
    Primary=(Weight=1,Axis=(X=1,Y=0,Z=0),Target=(X=-1,Y=0,Z=0),Kind=Location,Space="rope_aim_tgt")
    Secondary=(Weight=1,Axis=(X=0,Y=-1,Z=0),Target=(X=-1,Y=0,Z=0),Kind=Location,Space="rope_aim_up")
    Weight=1
    bPropagateToChildren=True
   ) {is=struct RigVMUnitNode}
   &RigUnit_BeginExecution()
  ]}
]
hierarchy [
 root {loc=<0 0 0>, rot=<0 0 0 1>}
 pelvis {parent=root, loc=<0 -0.7071 93.1191>, rot=<0.0127 0.707 -0.0127 -0.707>}
 spine_01 {parent=pelvis, loc=<6.7494 -0.6388 0>, rot=<0 0 0.034 -0.9994>}
 spine_02 {parent=spine_01, loc=<7.2518 -0.035 0>, rot=<0 0 0.0238 -0.9997>}
 spine_03 {parent=spine_02, loc=<6.0104 0 0>, rot=<0 0 0.0149 -0.9999>}
 spine_04 {parent=spine_03, loc=<8.1024 0 0>, rot=<0 0 -0.0132 -0.9999>}
 spine_05 {parent=spine_04, loc=<10.0146 0 0>, rot=<0 0 -0.0015 -1>}
 clavicle_l {parent=spine_05, loc=<4.3575 -0.1466 -2.3223>, rot=<0.0465 -0.7289 -0.009 -0.6829>}
 upperarm_l {parent=clavicle_l, loc=<15.5111 0 0>, rot=<0.0226 -0.3234 -0.0088 -0.9459>}
 lowerarm_l {parent=upperarm_l, loc=<26.6514 0 0>, rot=<0.0391 -0.0099 -0.2453 -0.9686>}
 hand_l {parent=lowerarm_l, loc=<22.3491 0 0>, rot=<-0.7081 -0.0507 0.0975 -0.6975>}
 index_metacarpal_l {parent=hand_l, loc=<1.6535 0.0183 -2.96>, rot=<0.0158 -0.0124 -0.0173 -0.9996>}
 index_01_l {parent=index_metacarpal_l, loc=<8.3008 0.0357 -0.0524>, rot=<-0.0044 0.0485 0.0924 -0.9945>}
 index_02_l {parent=index_01_l, loc=<3.3587 -0.3004 0.0452>, rot=<-0.002 0.0027 0.3225 -0.9466>}
 index_03_l {parent=index_02_l, loc=<2.5578 -0.123 -0.0252>, rot=<-0.002 0.0067 0.2558 -0.9667>}
 middle_metacarpal_l {parent=hand_l, loc=<2.1856 -0.1725 -0.5984>, rot=<0.006 0.0168 -0.0329 -0.9993>}
 middle_01_l {parent=middle_metacarpal_l, loc=<7.3275 -0.2447 0.0129>, rot=<-0.0208 0.0042 0.0858 -0.9961>}
 middle_02_l {parent=middle_01_l, loc=<4.0123 0.1401 0.1175>, rot=<-0.0218 0.0258 0.3512 -0.9357>}
 middle_03_l {parent=middle_02_l, loc=<2.6077 0.0534 0.1373>, rot=<0.0082 0.0188 0.1708 -0.9851>}
 pinky_metacarpal_l {parent=hand_l, loc=<1.7616 -0.9753 3.3814>, rot=<-0.0397 0.0814 -0.069 -0.9935>}
 pinky_01_l {parent=pinky_metacarpal_l, loc=<7.1709 -0.4831 -0.1094>, rot=<-0.0654 -0.0035 0.2242 -0.9723>}
 pinky_02_l {parent=pinky_01_l, loc=<2.8274 0.0893 -0.1043>, rot=<0.0037 0.0011 0.2854 -0.9584>}
 pinky_03_l {parent=pinky_02_l, loc=<2.2046 0.0307 -0.0078>, rot=<-0.0002 0.0098 0.1445 -0.9895>}
 ring_metacarpal_l {parent=hand_l, loc=<2.2074 -0.3625 1.3638>, rot=<-0.0284 0.0722 -0.0458 -0.9959>}
 ring_01_l {parent=ring_metacarpal_l, loc=<7.0295 -0.6197 -0.1446>, rot=<-0.0489 -0.0252 0.1892 -0.9804>}
 ring_02_l {parent=ring_01_l, loc=<3.6174 0.4646 -0.1027>, rot=<0 0 0.2806 -0.9598>}
 ring_03_l {parent=ring_02_l, loc=<2.4152 0.1141 0.0199>, rot=<0 0 0.1189 -0.9929>}
 thumb_01_l {parent=hand_l, loc=<1.3118 -0.3191 -2.459>, rot=<0.6546 -0.4042 0.0664 -0.6354>}
 thumb_02_l {parent=thumb_01_l, loc=<5.9336 -0.0448 0.0401>, rot=<-0.0037 0.0213 0.298 -0.9543>}
 thumb_03_l {parent=thumb_02_l, loc=<2.2223 0.3682 -0.0907>, rot=<0.0118 -0.013 0.0066 -0.9998>}
 lowerarm_twist_01_l {parent=lowerarm_l, loc=<17.8549 0 0>, rot=<0 0 0 1>}
 lowerarm_twist_02_l {parent=lowerarm_l, loc=<9.6199 0 0>, rot=<0 0 0 1>}
 upperarm_twist_01_l {parent=upperarm_l, loc=<0.0041 0 0>, rot=<0 0 0 -1>}
 upperarm_twist_02_l {parent=upperarm_l, loc=<13.7941 0 0>, rot=<0 0 0 -1>}
 clavicle_r {parent=spine_05, loc=<4.3575 -0.1466 2.3223>, rot=<0.7289 0.0465 -0.6829 0.009>}
 upperarm_r {parent=clavicle_r, loc=<-15.5111 0 0>, rot=<0.0226 -0.3234 -0.0088 -0.9459>}
 lowerarm_r {parent=upperarm_r, loc=<-26.6514 0 0>, rot=<0.0391 -0.0099 -0.2453 -0.9686>}
 hand_r {parent=lowerarm_r, loc=<-22.3492 0 0>, rot=<-0.7081 -0.0507 0.0975 -0.6975>}
 index_metacarpal_r {parent=hand_r, loc=<-1.6534 -0.0181 2.9601>, rot=<0.0158 -0.0124 -0.0173 -0.9996>}
 index_01_r {parent=index_metacarpal_r, loc=<-8.3008 -0.0357 0.0524>, rot=<-0.0044 0.0485 0.0924 -0.9945>}
 index_02_r {parent=index_01_r, loc=<-3.3589 0.3003 -0.0453>, rot=<-0.002 0.0027 0.3225 -0.9466>}
 index_03_r {parent=index_02_r, loc=<-2.5578 0.1231 0.0251>, rot=<-0.002 0.0067 0.2558 -0.9667>}
 middle_metacarpal_r {parent=hand_r, loc=<-2.1856 0.1726 0.5984>, rot=<0.006 0.0168 -0.0329 -0.9993>}
 middle_01_r {parent=middle_metacarpal_r, loc=<-7.3273 0.2447 -0.0129>, rot=<-0.0208 0.0042 0.0858 -0.9961>}
 middle_02_r {parent=middle_01_r, loc=<-4.0142 -0.1411 -0.07>, rot=<-0.0217 0.0259 0.3512 -0.9357>}
 middle_03_r {parent=middle_02_r, loc=<-2.6074 -0.0534 -0.1312>, rot=<0.0083 0.0188 0.1708 -0.9851>}
 pinky_metacarpal_r {parent=hand_r, loc=<-1.7617 0.9753 -3.3814>, rot=<-0.0397 0.0814 -0.069 -0.9935>}
 pinky_01_r {parent=pinky_metacarpal_r, loc=<-7.1707 0.4832 0.1094>, rot=<-0.0654 -0.0035 0.2242 -0.9723>}
 pinky_02_r {parent=pinky_01_r, loc=<-2.8274 -0.0892 0.1043>, rot=<0.0037 0.0011 0.2854 -0.9584>}
 pinky_03_r {parent=pinky_02_r, loc=<-2.2045 -0.031 0.0078>, rot=<-0.0002 0.0098 0.1445 -0.9895>}
 ring_metacarpal_r {parent=hand_r, loc=<-2.2073 0.3627 -1.3638>, rot=<-0.0284 0.0722 -0.0458 -0.9959>}
 ring_01_r {parent=ring_metacarpal_r, loc=<-7.0297 0.6197 0.1446>, rot=<-0.0489 -0.0252 0.1892 -0.9804>}
 ring_02_r {parent=ring_01_r, loc=<-3.6172 -0.4648 0.1027>, rot=<0 0 0.2806 -0.9598>}
 ring_03_r {parent=ring_02_r, loc=<-2.4152 -0.1141 -0.0199>, rot=<0 0 0.1189 -0.9929>}
 thumb_01_r {parent=hand_r, loc=<-1.3119 0.3192 2.459>, rot=<0.6546 -0.4042 0.0664 -0.6354>}
 thumb_02_r {parent=thumb_01_r, loc=<-5.9335 0.0447 -0.0401>, rot=<-0.0037 0.0213 0.298 -0.9543>}
 thumb_03_r {parent=thumb_02_r, loc=<-2.2223 -0.3682 0.0907>, rot=<0.0118 -0.013 0.0066 -0.9998>}
 lowerarm_twist_01_r {parent=lowerarm_r, loc=<-17.855 0 0>, rot=<0 0 0 1>}
 lowerarm_twist_02_r {parent=lowerarm_r, loc=<-9.6199 0 0>, rot=<0 0 0 1>}
 upperarm_twist_01_r {parent=upperarm_r, loc=<-0.0041 0 0>, rot=<0 0 0 1>}
 upperarm_twist_01_top_r {parent=upperarm_twist_01_r, loc=<0 0 0>, rot=<0 0 0 1>}
 upperarm_twist_01_bot_r {parent=upperarm_twist_01_r, loc=<0 0 0>, rot=<0 0 0 1>}
 upperarm_twist_02_r {parent=upperarm_r, loc=<-13.7942 0 0>, rot=<0 0 0 1>}
 armpouch_r_zero {parent=upperarm_r, loc=<-9.4836 -0.6271 -3.6822>, rot=<0.9988 0.011 0.0041 0.0469>}
 armpouch_r {parent=armpouch_r_zero, loc=<0 0 0>, rot=<0 0 0 -1>}
 armpouch_r_tgt {parent=armpouch_r_zero, loc=<-4.304 0 0>, rot=<0 0.0276 -0.0128 -0.9995>}
 preserverSquish_top_r_zero {parent=clavicle_r, loc=<-11.4494 -11.0149 0.538>, rot=<0.9991 0.0252 0.0336 0.0008>}
 preserverSquish_top_r {parent=preserverSquish_top_r_zero, loc=<0 0 0>, rot=<0 0 0 1>}
 neck_01 {parent=spine_05, loc=<11.3138 0.1529 0>, rot=<0 0 -0.2003 -0.9797>}
 neck_02 {parent=neck_01, loc=<4.5627 0 0>, rot=<0 0 0.0504 -0.9987>}
 collar_back_zero {parent=neck_02, loc=<-1.1779 -5.4826 0.0003>, rot=<-0.0991 -0.7001 0.0991 -0.7001>}
 collar_back {parent=collar_back_zero, loc=<0 0 0>, rot=<0 0 0 1>}
 preserver_l_zero {parent=spine_05, loc=<10.3794 1.0618 -11.326>, rot=<-0.0601 -0.0397 -0.321 -0.9443>}
 preserver_l {parent=preserver_l_zero, loc=<0 0 0>, rot=<0 0 0 -1>}
 preserver_l_scl {parent=preserver_l, loc=<0 0 0>, rot=<0 0 0 -1>}
 preserver_r_zero {parent=spine_05, loc=<10.3794 1.0618 11.326>, rot=<0.0601 0.0397 -0.321 -0.9443>}
 preserver_r {parent=preserver_r_zero, loc=<0 0 0>, rot=<0 0 0 -1>}
 preserver_r_scl {parent=preserver_r, loc=<0 0 0>, rot=<0 0 0 -1>}
 knife_zero {parent=spine_04, loc=<9.1372 14.4695 -9.5184>, rot=<-0.2072 -0.7547 -0.6206 -0.0488>}
 knife {parent=knife_zero, loc=<0 0 0>, rot=<0 0 0 1>}
 preserverSquish_bot_r_zero {parent=spine_04, loc=<4.2293 13.2011 12.6461>, rot=<0.0294 -0.7065 -0.0294 -0.7065>}
 preserverSquish_bot_r {parent=preserverSquish_bot_r_zero, loc=<0 0 0>, rot=<0 0 0 1>}
 pantsSide_l_zero {parent=spine_03, loc=<-7.7623 1.2366 -12.8958>, rot=<0.1642 -0.7846 0.5964 0.0415>}
 pantsSide_l {parent=pantsSide_l_zero, loc=<0 0 0>, rot=<0 0 0 -1>}
 stomachFlap_zero {parent=spine_03, loc=<0.8123 11.7819 0.0237>, rot=<-0.0093 0.0002 0.9948 0.101>}
 stomachFlap {parent=stomachFlap_zero, loc=<0 0 0>, rot=<0 0 0 1>}
 pantsSide_r_zero {parent=spine_02, loc=<-1.7117 1.467 12.8958>, rot=<0.1526 -0.787 -0.5957 -0.0503>}
 pantsSide_r {parent=pantsSide_r_zero, loc=<0 0 0>, rot=<0 0 0 -1>}
 thigh_l {parent=pelvis, loc=<-2.5693 -0.0109 -7.7501>, rot=<0.0012 0.0572 0.0155 -0.9982>}
 calf_l {parent=thigh_l, loc=<-39.0484 0 0>, rot=<0.0006 -0.0161 -0.0391 -0.9991>}
 calf_twist_01_l {parent=calf_l, loc=<-35.8131 -0.0967 0.0397>, rot=<0.0237 0.0095 -0.0035 -0.9997>}
 calf_twist_02_l {parent=calf_l, loc=<-18.8064 -0.0967 0.0397>, rot=<0.0237 0.0095 -0.0035 -0.9997>}
 foot_l {parent=calf_l, loc=<-42.561 0.1077 0.4292>, rot=<0.0418 -0.0455 0.0343 -0.9975>}
 ball_l {parent=foot_l, loc=<-6.1599 13.8264 0.6806>, rot=<0.0001 -0.0001 -0.7063 -0.7079>}
 thigh_twist_01_l {parent=thigh_l, loc=<-21.6303 0 0>, rot=<0.0013 0 0 -1>}
 thigh_r {parent=pelvis, loc=<-2.5691 -0.0109 7.7501>, rot=<0.0572 -0.0012 0.9982 0.0155>}
 calf_r {parent=thigh_r, loc=<39.0487 0 0>, rot=<0.0006 -0.0161 -0.0391 -0.9991>}
 calf_twist_01_r {parent=calf_r, loc=<35.8132 0.0967 -0.0397>, rot=<0.0237 0.0095 -0.0035 -0.9997>}
 calf_twist_02_r {parent=calf_r, loc=<18.8064 0.0967 -0.0397>, rot=<0.0237 0.0095 -0.0035 -0.9997>}
 foot_r {parent=calf_r, loc=<42.5611 -0.1077 -0.4292>, rot=<0.0418 -0.0455 0.0343 -0.9975>}
 ball_r {parent=foot_r, loc=<6.1598 -13.8264 -0.6806>, rot=<0.0001 -0.0001 -0.7063 -0.7079>}
 thigh_twist_01_r {parent=thigh_r, loc=<21.6303 0 0>, rot=<0.0013 0 0 -1>}
 rope_aim_tgt_zero {parent=thigh_twist_01_r, loc=<-6.2998 14.3018 8.3197>, rot=<-0.6655 0.0006 0.7464 0.0027>}
 rope_aim_tgt {parent=rope_aim_tgt_zero, loc=<0 0 0>, rot=<0 0 0 -1>}
 rope_aim_up_zero {parent=thigh_twist_01_r, loc=<-7.2447 -19.9454 8.9781>, rot=<-0.6655 0.0006 0.7464 0.0027>}
 rope_aim_up {parent=rope_aim_up_zero, loc=<0 0 0>, rot=<0 0 0 -1>}
 fannypack_zero {parent=pelvis, loc=<5.0831 -10.8839 0.5752>, rot=<-0.0165 -0.9992 0.0159 0.0325>}
 fannypack {parent=fannypack_zero, loc=<0 0 0>, rot=<0 0 0 -1>}
 fannypackHook_1_r_zero {parent=fannypack, loc=<2.2768 -3.7481 -11.0923>, rot=<-0.5638 0.0244 -0.0203 -0.8253>}
 fannypackHook_1_r {parent=fannypackHook_1_r_zero, loc=<0 0 0>, rot=<0 0 0 1>}
 pistol_zero {parent=fannypack, loc=<3.0873 -7.8464 -1.714>, rot=<-0.4933 -0.5322 -0.4818 0.4912>}
 pistol {parent=pistol_zero, loc=<0 0 0>, rot=<0 0 0 -1>}
 fannypackHook_2_r_zero {parent=pelvis, loc=<3.3788 -13.598 11.9181>, rot=<0.025 -0.5492 0.8314 0.081>}
 fannypackHook_2_r {parent=fannypackHook_2_r_zero, loc=<0 0 0>, rot=<0 0 0 -1>}
 ammo_1_l_zero {parent=pelvis, loc=<1.5474 -8.9038 -11.7312>, rot=<0.0145 -0.8773 -0.4797 0.0074>}
 ammo_1_l {parent=ammo_1_l_zero, loc=<0 0 0>, rot=<0 0 0 -1>}
 ammo_2_l_zero {parent=pelvis, loc=<1.593 -11.0231 -9.9502>, rot=<0.0153 -0.9481 -0.3172 0.0132>}
 ammo_2_l {parent=ammo_2_l_zero, loc=<0 0 0>, rot=<0 0 0 1>}
 pliers_zero {parent=pelvis, loc=<5.3266 -10.1288 12.974>, rot=<-0.0448 -0.9193 0.391 -0.0012>}
 pliers {parent=pliers_zero, loc=<0 0 0>, rot=<0 0 0 -1>}
 grenadePivot_1_l_zero {parent=pelvis, loc=<5.0542 -0.8261 -3.6015>, rot=<0.5 0 0 -0.866>}
 grenadePivot_1_l {parent=grenadePivot_1_l_zero, loc=<0 0 0>, rot=<0 0 0 -1>}
 grenade_1_l_zero {parent=grenadePivot_1_l, loc=<1.551 10.969 -0.4332>, rot=<0.0031 0.1733 0.9847 0.0176>}
 grenade_1_l {parent=grenade_1_l_zero, loc=<0 0 0>, rot=<0 0 0 -1>}
 grenadeHook_1_l_zero {parent=grenade_1_l, loc=<0.8531 -2.829 4.2984>, rot=<0.7769 0.0246 0.0518 -0.6269>}
 grenadeHook_1_l {parent=grenadeHook_1_l_zero, loc=<0 0 0>, rot=<0 0 0 1>}
 grenadePIvot_2_l_zero {parent=pelvis, loc=<5.0322 -1.4424 -4.3488>, rot=<0.5878 0 0 -0.809>}
 grenadePIvot_2_l {parent=grenadePIvot_2_l_zero, loc=<0 0 0>, rot=<0 0 0 -1>}
 grenade_2_l_zero {parent=grenadePIvot_2_l, loc=<1.6356 8.9697 -4.858>, rot=<-0.0067 0.2862 0.9581 0.01>}
 grenade_2_l {parent=grenade_2_l_zero, loc=<0 0 0>, rot=<0 0 0 1>}
 grenadeHook_2_l_zero {parent=grenade_2_l, loc=<0.9171 -6.4686 2.2345>, rot=<0.2637 0.0356 0.0298 -0.9635>}
 grenadeHook_2_l {parent=grenadeHook_2_l_zero, loc=<0 0 0>, rot=<0 0 0 -1>}
 carabinerPivot_zero {parent=pelvis, loc=<5.9955 -0.6936 3.6791>, rot=<0 0 0 -1>}
 carabinerPivot {parent=carabinerPivot_zero, loc=<0 0 0>, rot=<0 0 0 -1>}
 carabiner_zero {parent=carabinerPivot, loc=<-0.1147 3.974 9.1002>, rot=<0.0268 -0.3244 0.9445 0.0449>}
 carabiner {parent=carabiner_zero, loc=<0 0 0>, rot=<0 0 0 -1>}
 rope_aim_zero {parent=carabiner, loc=<5.3907 0.0937 2.156>, rot=<-0.2704 0.1689 -0.3506 -0.8806>}
 rope_aim {parent=rope_aim_zero, loc=<0 0 0>, rot=<0 0 0 -1>}
 rope_1_zero {parent=rope_aim, loc=<0 0 0>, rot=<0.6533 -0.2706 0.2706 -0.6533>}
 rope_1 {parent=rope_1_zero, loc=<0 0 0>, rot=<0 0 0 1>}
 rope_2_zero {parent=rope_1, loc=<1.3396 -0.0008 0.0018>, rot=<0 0 0 -1>}
 rope_2 {parent=rope_2_zero, loc=<0 0 0>, rot=<0 0 0 -1>}
 rope_3_zero {parent=rope_2, loc=<2.7761 -0.0016 0.0036>, rot=<0 0 0 -1>}
 rope_3 {parent=rope_3_zero, loc=<0 0 0>, rot=<0 0 0 -1>}
 pPlane1 {parent=rope_1, loc=<0 0 0>, rot=<0 0.3827 0 -0.9239>, scale=<24.2301 24.2301 24.2301>}
 carabiner_end {parent=carabiner_zero, loc=<5.3907 0.0937 2.156>, rot=<0 0 0 -1>}
]
