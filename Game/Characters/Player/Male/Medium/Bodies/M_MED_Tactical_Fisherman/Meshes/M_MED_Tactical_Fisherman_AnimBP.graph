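# Post-process style AnimBlueprint for the Tactical Fisherman player body, adding secondary motion and corrective deformation to its gear over a pose copied from the attached parent mesh.
# Spring-driven AnimDynamics swing pouches, grenades, pistol, knife, rope and flaps, while weighted constraints, a look-at and a Control Rig place the preserver, arm pouch, carabiner and rope.
# The event graph derives corrective alphas from arm and head rotations plus jump, crouch and targeting states to offset bones.
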AnimBlueprint /Game/Characters/Player/Male/Medium/Bodies/M_MED_Tactical_Fisherman/Meshes/M_MED_Tactical_Fisherman_AnimBP
target_skeleton=&/M_MED_Tactical_Fisherman_Skeleton
graphs [
 AnimGraph {nodes [
   n1 = Anim Dynamics - Bone: fannypack(ComponentPose=Local To Component(LocalPose=Copy Pose From Mesh() {bUseAttachedParent=true}), Alpha=1, ExternalForce="0,0,0") {
    angularDampingOverride=1
    angularSpringConstant=1500
    bAngularSpring=true
    bGravityOverrideInSimSpace=true
    bOverrideAngularDamping=true
    gravityScale=0.35
    simulationSpace=RootRelative
    alphaBoolBlend {blendOption=Linear}
    alphaScaleBiasClamp {scale=0.4}
    boundBone {boneName=fannypack}
    physicsBodyDefinitions [{
      sphereCollisionRadius=0
      boundBone {boneName=fannypack}
      boxExtents=<20 40 40>
      constraintSetup {angularLimitsMax=<0 0 0>, angularLimitsMin=<0 0 -40>, angularTarget=<0 0 0>}
      localJointOffset=<20 0 0>
     }]
   }
   n2 = Anim Dynamics - Bone: fannypackHook_1_r(ComponentPose=@n1, Alpha=1, ExternalForce="0,0,0") {
    angularDampingOverride=1
    angularSpringConstant=1500
    bAngularSpring=true
    bGravityOverrideInSimSpace=true
    bOverrideAngularDamping=true
    gravityScale=0.35
    simulationSpace=RootRelative
    alphaBoolBlend {blendOption=Linear}
    boundBone {boneName=fannypackHook_1_r}
    physicsBodyDefinitions [{
      sphereCollisionRadius=0
      boundBone {boneName=fannypackHook_1_r}
      boxExtents=<10 20 20>
      constraintSetup {angularLimitsMax=<0 20 0>, angularLimitsMin=<0 -20 -40>, angularTarget=<0 0 0>}
      localJointOffset=<10 0 0>
     }]
   }
   n3 = Anim Dynamics - Bone: fannypackHook_2_r(ComponentPose=@n2, Alpha=1, ExternalForce="0,0,0") {
    angularDampingOverride=1
    angularSpringConstant=1500
    bAngularSpring=true
    bGravityOverrideInSimSpace=true
    bOverrideAngularDamping=true
    gravityScale=0.35
    simulationSpace=RootRelative
    alphaBoolBlend {blendOption=Linear}
    boundBone {boneName=fannypackHook_2_r}
    physicsBodyDefinitions [{
      sphereCollisionRadius=0
      boundBone {boneName=fannypackHook_2_r}
      boxExtents=<10 20 20>
      constraintSetup {angularLimitsMax=<0 20 0>, angularLimitsMin=<0 -20 -40>, angularTarget=<0 0 0>}
      localJointOffset=<10 0 0>
     }]
   }
   n4 = Anim Dynamics - Bone: pistol(ComponentPose=@n3, Alpha=1, ExternalForce="0,0,0") {
    angularDampingOverride=1
    angularSpringConstant=1500
    bAngularSpring=true
    bGravityOverrideInSimSpace=true
    bOverrideAngularDamping=true
    gravityScale=0.35
    simulationSpace=RootRelative
    alphaBoolBlend {blendOption=Linear}
    alphaScaleBiasClamp {scale=0.3}
    boundBone {boneName=pistol}
    physicsBodyDefinitions [{
      sphereCollisionRadius=0
      boundBone {boneName=pistol}
      boxExtents=<20 10 20>
      constraintSetup {angularLimitsMax=<0 0 0>, angularLimitsMin=<0 0 -40>, angularTarget=<0 0 0>}
      localJointOffset=<0 15 0>
     }]
   }
   n5 = Anim Dynamics - Bone: pliers(ComponentPose=@n4, Alpha=1, ExternalForce="0,0,0") {
    angularDampingOverride=1
    angularSpringConstant=1500
    bAngularSpring=true
    bGravityOverrideInSimSpace=true
    bOverrideAngularDamping=true
    gravityScale=0.35
    simulationSpace=RootRelative
    alphaBoolBlend {blendOption=Linear}
    alphaScaleBiasClamp {scale=0.3}
    boundBone {boneName=pliers}
    physicsBodyDefinitions [{
      sphereCollisionRadius=0
      boundBone {boneName=pliers}
      boxExtents=<10 20 20>
      constraintSetup {angularLimitsMax=<0 10 0>, angularLimitsMin=<0 -10 -40>, angularTarget=<0 0 0>}
      localJointOffset=<15 0 0>
     }]
   }
   n6 = Get crouch_alpha() -> (crouch_alpha:double)
   n7 = "Transform (Modify) Bone - Bone: grenade_1_l_zero"(ComponentPose=@n5, Alpha=@n6, Translation="0,0,0", Rotation="0,-20,-40", Scale="1,1,1") {
    rotationMode=BMM_Additive
    rotationSpace=BCS_BoneSpace
    translationMode=BMM_Additive
    translationSpace=BCS_BoneSpace
    alphaBoolBlend {blendOption=Linear}
    boneToModify {boneName=grenade_1_l_zero}
    rotation {roll=-40, yaw=-20}
   }
   n8 = "Transform (Modify) Bone - Bone: grenadePivot_1_l"(ComponentPose=@n7, Alpha=@n6, Translation="0,0,0", Rotation="0,0,8", Scale="1,1,1") {
    rotationMode=BMM_Additive
    rotationSpace=BCS_BoneSpace
    alphaBoolBlend {blendOption=Linear}
    alphaScaleBiasClamp {bClampResult=true}
    boneToModify {boneName=grenadePivot_1_l}
    rotation {roll=8}
   }
   n9 = "Transform (Modify) Bone - Bone: grenadePIvot_2_l"(ComponentPose=@n8, Alpha=@n6, Translation="0,0,0", Rotation="0,0,8", Scale="1,1,1") {
    rotationMode=BMM_Additive
    rotationSpace=BCS_BoneSpace
    alphaBoolBlend {blendOption=Linear}
    alphaScaleBiasClamp {bClampResult=true}
    boneToModify {boneName=grenadePIvot_2_l}
    rotation {roll=8}
   }
   n10 = Anim Dynamics - Bone: grenade_1_l(ComponentPose=@n9, Alpha=1, ExternalForce="0,0,0") {
    angularDampingOverride=1
    angularSpringConstant=1500
    bAngularSpring=true
    bGravityOverrideInSimSpace=true
    bOverrideAngularDamping=true
    gravityScale=0.35
    simulationSpace=RootRelative
    alphaBoolBlend {blendOption=Linear}
    alphaScaleBiasClamp {scale=0.6}
    boundBone {boneName=grenade_1_l}
    physicsBodyDefinitions [{
      sphereCollisionRadius=0
      boundBone {boneName=grenade_1_l}
      boxExtents=<10 20 20>
      constraintSetup {angularLimitsMax=<0 20 0>, angularLimitsMin=<0 -20 -40>, angularTarget=<0 0 0>}
      localJointOffset=<10 0 0>
     }]
   }
   n11 = Anim Dynamics - Bone: grenadeHook_1_l(ComponentPose=@n10, Alpha=1, ExternalForce="0,0,0") {
    angularDampingOverride=1
    angularSpringConstant=1500
    bAngularSpring=true
    bGravityOverrideInSimSpace=true
    bOverrideAngularDamping=true
    gravityScale=0.35
    simulationSpace=RootRelative
    alphaBoolBlend {blendOption=Linear}
    boundBone {boneName=grenadeHook_1_l}
    componentLinearAccScale {z=0.5}
    componentLinearVelScale {z=0.5}
    physicsBodyDefinitions [{
      sphereCollisionRadius=0
      boundBone {boneName=grenadeHook_1_l}
      boxExtents=<10 20 20>
      constraintSetup {angularLimitsMax=<0 20 0>, angularLimitsMin=<0 -20 -40>, angularTarget=<0 0 0>}
      localJointOffset=<10 0 0>
     }]
   }
   n12 = Anim Dynamics - Bone: grenade_2_l(ComponentPose=@n11, Alpha=1, ExternalForce="0,0,0") {
    angularDampingOverride=1
    angularSpringConstant=1500
    bAngularSpring=true
    bGravityOverrideInSimSpace=true
    bOverrideAngularDamping=true
    gravityScale=0.35
    simulationSpace=RootRelative
    alphaBoolBlend {blendOption=Linear}
    alphaScaleBiasClamp {scale=0.6}
    boundBone {boneName=grenade_2_l}
    physicsBodyDefinitions [{
      sphereCollisionRadius=0
      boundBone {boneName=grenade_2_l}
      boxExtents=<10 20 20>
      constraintSetup {angularLimitsMax=<0 20 0>, angularLimitsMin=<0 -20 -40>, angularTarget=<0 0 0>}
      localJointOffset=<10 0 0>
     }]
   }
   n13 = Anim Dynamics - Bone: grenadeHook_2_l(ComponentPose=@n12, Alpha=1, ExternalForce="0,0,0") {
    angularDampingOverride=1
    angularSpringConstant=1500
    bAngularSpring=true
    bGravityOverrideInSimSpace=true
    bOverrideAngularDamping=true
    gravityScale=0.35
    simulationSpace=RootRelative
    alphaBoolBlend {blendOption=Linear}
    boundBone {boneName=grenadeHook_2_l}
    componentLinearAccScale {z=0.5}
    componentLinearVelScale {z=0.5}
    physicsBodyDefinitions [{
      sphereCollisionRadius=0
      boundBone {boneName=grenadeHook_2_l}
      boxExtents=<10 20 20>
      constraintSetup {angularLimitsMax=<0 20 0>, angularLimitsMin=<0 -20 -40>, angularTarget=<0 0 0>}
      localJointOffset=<10 0 0>
     }]
   }
   n14 = Anim Dynamics - Bone: ammo_1_l(ComponentPose=@n13, Alpha=1, ExternalForce="0,0,0") {
    angularDampingOverride=1
    angularSpringConstant=1500
    bAngularSpring=true
    bGravityOverrideInSimSpace=true
    bOverrideAngularDamping=true
    gravityScale=0.35
    simulationSpace=RootRelative
    alphaBoolBlend {blendOption=Linear}
    alphaScaleBiasClamp {scale=0.6}
    boundBone {boneName=ammo_1_l}
    physicsBodyDefinitions [{
      sphereCollisionRadius=0
      boundBone {boneName=ammo_1_l}
      boxExtents=<5 10 10>
      constraintSetup {angularLimitsMax=<0 5 0>, angularLimitsMin=<0 -5 -40>, angularTarget=<0 0 0>}
      localJointOffset=<5 0 0>
     }]
   }
   n15 = Anim Dynamics - Bone: ammo_2_l(ComponentPose=@n14, Alpha=1, ExternalForce="0,0,0") {
    angularDampingOverride=1
    angularSpringConstant=1500
    bAngularSpring=true
    bGravityOverrideInSimSpace=true
    bOverrideAngularDamping=true
    gravityScale=0.35
    simulationSpace=RootRelative
    alphaBoolBlend {blendOption=Linear}
    alphaScaleBiasClamp {scale=0.6}
    boundBone {boneName=ammo_2_l}
    physicsBodyDefinitions [{
      sphereCollisionRadius=0
      boundBone {boneName=ammo_2_l}
      boxExtents=<5 10 10>
      constraintSetup {angularLimitsMax=<0 5 0>, angularLimitsMin=<0 -5 -40>, angularTarget=<0 0 0>}
      localJointOffset=<7 0 0>
     }]
   }
   n16 = "Constraint - preserver_l_zero (PP)"!(ComponentPose=@n15, Alpha=1, ConstraintWeights_0=0.5, ConstraintWeights_1:float=0.5) -> (Pose) {
    alphaBoolBlend {blendOption=Linear}
    boneToModify {boneName=preserver_l_zero}
    constraintSetup [{targetBone {boneName=spine_05}}, {targetBone {boneName=clavicle_l}}]
    constraintWeights [1,1]
   }
   n17 = "Constraint - preserver_r_zero (PP)"!(ComponentPose=@n16, Alpha=0.3, ConstraintWeights_0=0.5, ConstraintWeights_1:float=0.5) -> (Pose) {
    alphaBoolBlend {blendOption=Linear}
    boneToModify {boneName=preserver_r_zero}
    constraintSetup [{targetBone {boneName=spine_05}}, {targetBone {boneName=clavicle_r}}]
    constraintWeights [1,1]
   }
   n18 = "Transform (Modify) Bone - Bone: preserver_l_scl"(ComponentPose=@n17, Alpha=Get arm_up_l() -> (arm_up_l:double), Translation="7,0,0", Rotation="0,0,0", Scale="1,1,1") {
    translationMode=BMM_Additive
    translationSpace=BCS_BoneSpace
    alphaBoolBlend {blendOption=Linear}
    alphaScaleBiasClamp {scale=0}
    boneToModify {boneName=preserver_l_scl}
    translation {x=7}
   }
   n19 = Get arm_up_r() -> (arm_up_r:double)
   n20 = "Transform (Modify) Bone - Bone: preserver_r_scl"(ComponentPose=@n18, Alpha=@n19, Translation="13,0,0", Rotation="0,0,0", Scale="1,1,1") {
    translationMode=BMM_Additive
    translationSpace=BCS_BoneSpace
    alphaBoolBlend {blendOption=Linear}
    alphaScaleBiasClamp {scale=0}
    boneToModify {boneName=preserver_r_scl}
    translation {x=13}
   }
   n21 = "Transform (Modify) Bone - Bone: collar_back"(ComponentPose=@n20, Alpha=Get head_rot_back() -> (head_rot_back:double), Translation="-5,0,0", Rotation="0,0,-35", Scale="1,1,1") {
    rotationMode=BMM_Additive
    rotationSpace=BCS_BoneSpace
    translationMode=BMM_Additive
    translationSpace=BCS_BoneSpace
    alphaBoolBlend {blendOption=Linear}
    boneToModify {boneName=collar_back}
    rotation {roll=-35}
    translation {x=-5}
   }
   n22 = "Transform (Modify) Bone - Bone: preserverSquish_top_r"(ComponentPose=@n21, Alpha=@n19, Translation="4,0,0", Rotation="0,0,0", Scale="1,1,1") {
    translationMode=BMM_Additive
    translationSpace=BCS_BoneSpace
    alphaBoolBlend {blendOption=Linear}
    boneToModify {boneName=preserverSquish_top_r}
    translation {x=4}
   }
   n23 = "Transform (Modify) Bone - Bone: upperarm_twist_01_bot_r"(ComponentPose=@n22, Alpha=@n19, Translation="0,0,5", Rotation="0,0,0", Scale="1,1,1") {
    translationMode=BMM_Additive
    translationSpace=BCS_ParentBoneSpace
    alphaBoolBlend {blendOption=Linear}
    boneToModify {boneName=upperarm_twist_01_bot_r}
    translation {z=5}
   }
   n24 = "Transform (Modify) Bone - Bone: upperarm_twist_01_top_r"(ComponentPose=@n23, Alpha=@n19, Translation="0,0,-3", Rotation="0,0,0", Scale="1,1,1") {
    translationMode=BMM_Additive
    translationSpace=BCS_ParentBoneSpace
    alphaBoolBlend {blendOption=Linear}
    boneToModify {boneName=upperarm_twist_01_top_r}
    translation {z=-3}
   }
   n25 = "Constraint - armpouch_r_zero (PPPPPPPP)"!(
    ComponentPose=@n24
    Alpha=1
    ConstraintWeights_0=0.3558
    ConstraintWeights_1:float=0.1898
    ConstraintWeights_2:float=0.0328
    ConstraintWeights_3:float=0.3044
    ConstraintWeights_4:float=0.0001
    ConstraintWeights_5:float=0
    ConstraintWeights_6:float=0.001
    ConstraintWeights_7:float=0.1161
   ) -> (Pose) {
    alphaBoolBlend {blendOption=Linear}
    boneToModify {boneName=armpouch_r_zero}
    constraintSetup [
     {targetBone {boneName=upperarm_twist_01_top_r}}
     {targetBone {boneName=upperarm_r}}
     {targetBone {boneName=upperarm_twist_01_bot_r}}
     {targetBone {boneName=upperarm_twist_02_r}}
     {targetBone {boneName=spine_03}}
     {targetBone {boneName=spine_04}}
     {targetBone {boneName=spine_05}}
     {targetBone {boneName=clavicle_r}}
    ]
    constraintWeights [1,1,1,1,1,1,1,1]
   }
   n26 = "Constraint - armpouch_r_tgt (PPPPP)"!(
    ComponentPose=@n25
    Alpha=1
    ConstraintWeights_0=0.0497
    ConstraintWeights_1:float=0.3124
    ConstraintWeights_2:float=0.0306
    ConstraintWeights_3:float=0.255
    ConstraintWeights_4:float=0.3523
   ) -> (Pose) {
    alphaBoolBlend {blendOption=Linear}
    boneToModify {boneName=armpouch_r_tgt}
    constraintSetup [
     {targetBone {boneName=clavicle_r}}
     {targetBone {boneName=upperarm_twist_01_top_r}}
     {targetBone {boneName=upperarm_twist_01_bot_r}}
     {targetBone {boneName=upperarm_r}}
     {targetBone {boneName=upperarm_twist_02_r}}
    ]
    constraintWeights [1,1,1,1,1]
   }
   n27 = Look At - Bone: armpouch_r(ComponentPose=@n26, Alpha=1) {
    alphaBoolBlend {blendOption=Linear}
    boneToModify {boneName=armpouch_r}
    lookAtTarget {boneReference {boneName=armpouch_r_tgt}}
    lookAt_Axis {axis=<1 0>}
   }
   n28 = Anim Dynamics - Bone: pantsSide_l(ComponentPose=@n27, Alpha=1, ExternalForce="0,0,0") {
    angularDampingOverride=1
    angularSpringConstant=1500
    bAngularSpring=true
    bGravityOverrideInSimSpace=true
    bOverrideAngularDamping=true
    gravityScale=0.35
    simulationSpace=RootRelative
    alphaBoolBlend {blendOption=Linear}
    alphaScaleBiasClamp {scale=0.3}
    boundBone {boneName=pantsSide_l}
    physicsBodyDefinitions [{
      sphereCollisionRadius=0
      boundBone {boneName=pantsSide_l}
      boxExtents=<5 10 10>
      constraintSetup {angularLimitsMax=<0 0 40>, angularLimitsMin=<0 0 -40>, angularTarget=<0 0 0>}
      localJointOffset=<10 0 0>
     }]
   }
   n29 = Anim Dynamics - Bone: pantsSide_r(ComponentPose=@n28, Alpha=1, ExternalForce="0,0,0") {
    angularDampingOverride=1
    angularSpringConstant=1500
    bAngularSpring=true
    bGravityOverrideInSimSpace=true
    bOverrideAngularDamping=true
    gravityScale=0.35
    simulationSpace=RootRelative
    alphaBoolBlend {blendOption=Linear}
    alphaScaleBiasClamp {scale=0.3}
    boundBone {boneName=pantsSide_r}
    physicsBodyDefinitions [{
      sphereCollisionRadius=0
      boundBone {boneName=pantsSide_r}
      boxExtents=<5 10 10>
      constraintSetup {angularLimitsMax=<0 0 40>, angularLimitsMin=<0 0 -40>, angularTarget=<0 0 0>}
      localJointOffset=<10 0 0>
     }]
   }
   n30 = Anim Dynamics - Bone: stomachFlap(ComponentPose=@n29, Alpha=1, ExternalForce="0,0,0") {
    angularDampingOverride=1
    angularSpringConstant=1500
    bAngularSpring=true
    bGravityOverrideInSimSpace=true
    bOverrideAngularDamping=true
    gravityScale=0.35
    simulationSpace=RootRelative
    alphaBoolBlend {blendOption=Linear}
    boundBone {boneName=stomachFlap}
    physicsBodyDefinitions [{
      sphereCollisionRadius=0
      boundBone {boneName=stomachFlap}
      boxExtents=<10 20 20>
      constraintSetup {angularLimitsMax=<0 0 0>, angularLimitsMin=<0 0 -40>, angularTarget=<0 0 0>}
      localJointOffset=<10 0 0>
     }]
   }
   n31 = "Constraint - carabiner_zero (PPP)"!(ComponentPose=@n30, Alpha=1, ConstraintWeights_0=0.3193, ConstraintWeights_1:float=0.0187, ConstraintWeights_2:float=0.662) -> (Pose) {
    alphaBoolBlend {blendOption=Linear}
    boneToModify {boneName=carabiner_zero}
    constraintSetup [{targetBone {boneName=spine_01}}, {targetBone {boneName=thigh_twist_01_r}}, {targetBone {boneName=pelvis}}]
    constraintWeights [1,1,1]
   }
   n32 = "Transform (Modify) Bone - Bone: carabiner"(ComponentPose=@n31, Alpha=1, Translation="0,0,0", Rotation="20,0,0", Scale="1,1,1") {
    rotationMode=BMM_Additive
    rotationSpace=BCS_BoneSpace
    alphaBoolBlend {blendOption=Linear}
    boneToModify {boneName=carabiner}
    rotation {pitch=20}
   }
   n33 = "Transform (Modify) Bone - Bone: rope_aim_zero"(ComponentPose=@n32, Alpha=Get jump_alpha() -> (jump_alpha:double), Translation="0,0,1.5", Rotation="0,0,0", Scale="1,1,1") {
    rotationMode=BMM_Additive
    rotationSpace=BCS_BoneSpace
    translationMode=BMM_Additive
    translationSpace=BCS_ParentBoneSpace
    alphaBoolBlend {blendOption=Linear}
    alphaScaleBiasClamp {bInterpResult=true, interpSpeedDecreasing=5, interpSpeedIncreasing=5}
    boneToModify {boneName=rope_aim_zero}
    translation {z=1.5}
   }
   n34 = "Constraint - rope_aim_tgt (PPP)"!(ComponentPose=@n33, Alpha=1, ConstraintWeights_0=0.0873, ConstraintWeights_1:float=0.099, ConstraintWeights_2:float=0.8138) -> (Pose) {
    alphaBoolBlend {blendOption=Linear}
    boneToModify {boneName=rope_aim_tgt}
    constraintSetup [{targetBone {boneName=thigh_r}}, {targetBone {boneName=pelvis}}, {targetBone {boneName=thigh_twist_01_r}}]
    constraintWeights [1,1,1]
   }
   n35 = "Constraint - rope_aim_up (PPP)"!(ComponentPose=@n34, Alpha=1, ConstraintWeights_0=0.0274, ConstraintWeights_1:float=0.0167, ConstraintWeights_2:float=0.9558) -> (Pose) {
    alphaBoolBlend {blendOption=Linear}
    boneToModify {boneName=rope_aim_up}
    constraintSetup [{targetBone {boneName=thigh_r}}, {targetBone {boneName=pelvis}}, {targetBone {boneName=thigh_twist_01_r}}]
    constraintWeights [1,1,1]
   }
   n36 = Control Rig(Source=Component To Local(ComponentPose=@n35)) {
    alphaBoolBlend {blendOption=Linear}
    controlRigAssetReference {blueprintRigClass=&/M_MED_Tactical_Fisherman_CtrlRig.M_MED_Tactical_Fisherman_CtrlRig_C}
   }
   n37 = Anim Dynamics - Bone: rope_1(ComponentPose=Local To Component(LocalPose=@n36), Alpha=1, ExternalForce="0,0,0") {
    angularDampingOverride=1
    angularSpringConstant=1500
    bAngularSpring=true
    bGravityOverrideInSimSpace=true
    bOverrideAngularDamping=true
    gravityScale=0.35
    simulationSpace=RootRelative
    alphaBoolBlend {blendOption=Linear}
    boundBone {boneName=rope_1}
    componentLinearAccScale {z=0.4}
    physicsBodyDefinitions [{
      sphereCollisionRadius=0
      boundBone {boneName=rope_1}
      boxExtents=<10 20 20>
      constraintSetup {angularLimitsMax=<0 0 0>, angularLimitsMin=<0 0 -40>, angularTarget=<0 0 0>}
      localJointOffset=<10 0 0>
     }]
   }
   n38 = Anim Dynamics - Bone: rope_3(ComponentPose=@n37, Alpha=1, ExternalForce="0,0,0") {
    angularDampingOverride=1
    angularSpringConstant=1500
    bAngularSpring=true
    bGravityOverrideInSimSpace=true
    bOverrideAngularDamping=true
    gravityScale=0.35
    simulationSpace=RootRelative
    alphaBoolBlend {blendOption=Linear}
    boundBone {boneName=rope_3}
    componentLinearAccScale {z=0.4}
    physicsBodyDefinitions [{
      sphereCollisionRadius=0
      boundBone {boneName=rope_3}
      boxExtents=<10 20 20>
      constraintSetup {angularLimitsMax=<10 40 0>, angularLimitsMin=<-10 -40 -50>, angularTarget=<0 0 0>}
      localJointOffset=<10 0 0>
     }]
   }
   n39 = "Constraint - carabiner_end (T)"(ComponentPose=@n38, Alpha=1, ConstraintWeights_0=1) {
    alphaBoolBlend {blendOption=Linear}
    boneToModify {boneName=carabiner_end}
    constraintSetup [{transformType=Translation, targetBone {boneName=rope_1}}]
    constraintWeights [1]
   }
   n40 = "Constraint - knife_zero (PPPPPPP)"!(
    ComponentPose=@n39
    Alpha=1
    ConstraintWeights_0=0.1303
    ConstraintWeights_1:float=0.014
    ConstraintWeights_2:float=0.0898
    ConstraintWeights_3:float=0.1812
    ConstraintWeights_4:float=0.3049
    ConstraintWeights_5:float=0.2695
    ConstraintWeights_6:float=0.0103
   ) -> (Pose) {
    alphaBoolBlend {blendOption=Linear}
    boneToModify {boneName=knife_zero}
    constraintSetup [
     {targetBone {boneName=clavicle_l}}
     {targetBone {boneName=spine_01}}
     {targetBone {boneName=spine_02}}
     {targetBone {boneName=spine_03}}
     {targetBone {boneName=spine_04}}
     {targetBone {boneName=spine_05}}
     {targetBone {boneName=upperarm_twist_01_l}}
    ]
    constraintWeights [1,1,1,1,1,1,1]
   }
   n41 = "Transform (Modify) Bone - Bone: knife_zero"(ComponentPose=@n40, Alpha=Get target_alpha() -> (target_alpha:double), Translation="0,0,0", Rotation="20,0,0", Scale="1,1,1") {
    rotationMode=BMM_Additive
    alphaBoolBlend {blendOption=Linear}
    alphaScaleBiasClamp {bInterpResult=true}
    boneToModify {boneName=knife_zero}
    rotation {pitch=20}
   }
   n42 = Anim Dynamics - Bone: knife(ComponentPose=@n41, Alpha=1, ExternalForce="0,0,0") {
    angularDampingOverride=1
    angularSpringConstant=1500
    bAngularSpring=true
    bGravityOverrideInSimSpace=true
    bOverrideAngularDamping=true
    gravityScale=0.35
    simulationSpace=RootRelative
    alphaBoolBlend {blendOption=Linear}
    alphaScaleBiasClamp {scale=0.1}
    boundBone {boneName=knife}
    physicsBodyDefinitions [{
      sphereCollisionRadius=0
      boundBone {boneName=knife}
      boxExtents=<10 20 20>
      constraintSetup {angularLimitsMax=<0 45 0>, angularLimitsMin=<0 -45 -80>, angularTarget=<0 0 0>}
      localJointOffset=<15 0 0>
     }]
   }
   Output Pose(Result=Component To Local(ComponentPose=@n42)) {layerGroup=None}
  ]}
 EventGraph {nodes [
   Event BlueprintUpdateAnimation()
   > n1 = Sequence!(execute) -> (then_0, then_1, then_2:Exec, then_3:Exec)
   > Set jump_alpha(jump_alpha:double="To Float (Boolean)"(InBool=Get bIsJumping())) -> (Output_Get:double)
   n2 = BreakTransform(InTransform=GetDeltaTransformFromRefPose(self=GetOwnerSkeletalMeshComponent(), BoneName=upperarm_l, BaseName=spine_05))
   Set arm_up_l(execute=@n1.then_2, arm_up_l:double=MapRangeClamped(Value=BreakRotator(InRot=@n2.Rotation).Yaw, InRangeA=40, InRangeB=100)) -> (Output_Get:double)
   n3 = BreakTransform(InTransform=GetDeltaTransformFromRefPose(self=GetOwnerSkeletalMeshComponent(), BoneName=upperarm_r, BaseName=spine_05))
   Set arm_up_r(execute=@n1.then_1, arm_up_r:double=MapRangeClamped(Value=BreakRotator(InRot=@n3.Rotation).Yaw, InRangeA=85, InRangeB=125)) -> (Output_Get:double)
   n4 = BreakTransform(InTransform=GetDeltaTransformFromRefPose(self=GetOwnerSkeletalMeshComponent(), BoneName=head, BaseName=spine_05))
   Set head_rot_back(execute=@n1.then_3, head_rot_back:double=MapRangeClamped(Value=BreakRotator(InRot=@n4.Rotation).Yaw, InRangeA=0, InRangeB=-55)) -> (Output_Get:double)
   Event OnEndTargeting() {is=title K2Node_Event}
   > Set target_alpha(target_alpha:double=0) -> (Output_Get:double)
   Event OnBeginTargeting() {is=title K2Node_Event}
   > Set target_alpha(target_alpha=1)
   Event OnBeginCrouching()
   > Set crouch_alpha(crouch_alpha:double=1) -> (Output_Get:double)
   Event OnEndCrouching()
   > Set crouch_alpha(crouch_alpha=0)
  ]}
]
variables [
 arm_up_l {type=number}
 arm_up_r {type=number}
 crouch_alpha {type=number}
 head_rot_back {type=number}
 jump_alpha {type=number}
 target_alpha {type=number}
]
